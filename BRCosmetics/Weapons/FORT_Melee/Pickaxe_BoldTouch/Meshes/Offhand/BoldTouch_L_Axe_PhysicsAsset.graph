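# Physics asset for the left offhand axe mesh of the BoldTouch pickaxe, a Battle Royale melee weapon cosmetic.
# A single capsule body on the b_AR_Root bone gives the mesh one simplified collision shape, slightly offset and rotated from the bone.
# It defines no constraints, so there are no jointed or separately simulated parts.
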PhysicsAsset /BRCosmetics/Weapons/FORT_Melee/Pickaxe_BoldTouch/Meshes/Offhand/BoldTouch_L_Axe_PhysicsAsset
bodies {b_AR_Root [{shapeType=Capsule, length=17.2773, radius=15.5799, center=<-0.0393 -2.285 12.9548>, rotation=<rot -8.5788 -90 -0.1194>}]}
constraints []
